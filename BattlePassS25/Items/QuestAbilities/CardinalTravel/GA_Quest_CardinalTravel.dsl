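; Quest gameplay ability for Cardinal Travel that reports how far the owning player travels north-east.
; Tracking is gated on the skydive-from-bus ended event or the match already being in the SafeZones phase.
; A looping timer of roughly one second, with a randomized interval, compares current and previous XY positions.
; Movement within a fixed heading window broadcasts a Travel.NorthEast custom verb with floored distance in meters, and every check resets the previous position.
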(event Ability|EventOnEndAbility
  (params (bWasCancelled :type Boolean))
  (Ability|EndAbility))

(event Ability|EventFailedToActivatePassiveAbility
  (params (ActorInfo :type Gameplay%20Ability%20Actor%20Info%20Structure))
  (Ability|EndAbility))

(event Ability|EventActivateAbility
  (|CallParentFunction)
  (Ability|Tasks|WaitGameplayEvent "(TagName=\"Event.SkydiveEnded.FromBus\")" 0 true
    (:then
      (bind _newparam (CallFunction|AlreadyinSafeZonePhase))
      (if _newparam
        _))
    (:EventReceived
      _))
  (CallFunction|GetLocation)
  (bind _returnvalue (Utilities|Time|SetTimerbyEvent (delegate Custom|Check%20Direction) (Math|Random|RandomFloatinRange 0.9 1.1) true))
  (Variables|Default|SetTimer%20Handle _returnvalue))

(event Custom|Check%20Direction
  (bind _current_loc (Variables|Default|GetCurrent%20Loc))
  (bind _y (.y _current_loc))
  (bind _x (.x _current_loc))
  (bind _returnvalue (Math|Vector|MakeVector _x _y))
  (bind _previous_loc (Variables|Default|GetPrevious%20Loc))
  (bind _y_2 (.y _previous_loc))
  (bind _x_2 (.x _previous_loc))
  (bind _returnvalue_2 (Math|Vector|MakeVector _x_2 _y_2))
  (Variables|Default|SetCurrent%20Loc (Transformation|GetActorLocation (Variables|GAVars|Data|GetOwningPlayerPawn)))
  (if (and (<= (Fortnite|VectortoNormalizedAngleinDegrees (Math|Vector|GetUnitDirection(Vector) _returnvalue _returnvalue_2)) -104.0) (>= (Fortnite|VectortoNormalizedAngleinDegrees (Math|Vector|GetUnitDirection(Vector) _returnvalue _returnvalue_2)) -164.0))
    (CustomVerb|BroadcastCustomVerbMessage (CustomVerb|MakeCustomVerbMessage (Variables|GAVars|Data|GetOwningPlayerPawn) "(TagName=\"CustomVerb.Travel.NorthEast\")" (Math|Float|Floor (/ (Math|Vector|Distance(Vector) _returnvalue _returnvalue_2) 1e+02))))
    (CallFunction|GetLocation)
    (else
      (CallFunction|GetLocation))))

(fn AlreadyInSafeZonePhase
  (return (Utilities|Enum|Equal(Enum) (BattleRoyaleGamePhaseLogic|GetGamePhase (BattleRoyaleGamePhaseLogic|GetBattleRoyaleGamePhaseLogicComponent)) "SafeZones")))

(fn GetLocation
  (Variables|Default|SetPrevious%20Loc (Math|Vector|MakeVector (.x (Transformation|GetActorLocation (Variables|GAVars|Data|GetOwningPlayerPawn))) (.y (Transformation|GetActorLocation (Variables|GAVars|Data|GetOwningPlayerPawn)))))
  (return))
