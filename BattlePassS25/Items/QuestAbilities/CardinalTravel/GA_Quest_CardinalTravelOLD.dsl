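; Quest gameplay ability, named as an old version, that tracks how far the player pawn travels toward the northeast.
; On activation it stores the pawn's XY position and starts a looping timer of roughly one randomized second.
; Each tick compares positions, broadcasting a NorthEast travel verb with truncated distance when neither axis decreases, and resets the reference only if X has not decreased.
; Only that direction is handled; the angle helpers are unused, one empty, and a debug print remains.
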(event Ability|EventActivateAbilityFromEvent
  (params (EventData :type Gameplay%20Event%20Data%20Structure%20%28by%20ref%29))
  (|CallParentFunction EventData)
  (CallFunction|GetLocation)
  (bind _returnvalue (Utilities|Time|SetTimerbyEvent (delegate Custom|Check%20Direction) (Math|Random|RandomFloatinRange 0.9 1.1) true))
  (Variables|Default|SetTimer%20Handle _returnvalue))

(event Custom|Check%20Direction
  (bind _current_loc (Variables|Default|GetCurrent%20Loc))
  (bind _x (.x _current_loc))
  (bind _y (.y _current_loc))
  (bind _returnvalue (Math|Vector|MakeVector _x _y))
  (bind _xy_vector (Variables|Default|GetXY%20Vector))
  (bind _y_2 (.y _xy_vector))
  (bind _x_2 (.x _xy_vector))
  (bind _returnvalue_2 (Math|Vector|MakeVector _x_2 _y_2))
  (bind _returnvalue_3 (Math|Vector|Distance(Vector) _returnvalue _returnvalue_2))
  (bind _returnvalue_4 (Math|Float|Truncate _returnvalue_3))
  (Variables|Default|SetCurrent%20Loc (Transformation|GetActorLocation (Variables|GAVars|Data|GetOwningPlayerPawn)))
  (if (>= _x _x_2)
    (if (>= _y _y_2)
      (Development|PrintString (Utilities|String|ToString(Integer) _returnvalue_4))
      (CustomVerb|BroadcastCustomVerbMessage (CustomVerb|MakeCustomVerbMessage (Variables|GAVars|Data|GetOwningPlayerPawn) "(TagName=\"CustomVerb.Travel.NorthEast\")" _returnvalue_4))
      (CallFunction|GetLocation)
      (else
        (CallFunction|GetLocation)))))

(fn AngleBetweenVectors
  (params (Destination%20Rotation :type Vector) (Origin%20Rotation :type Vector) (Rotation%20Normal :type Vector))
  (return (Math|Float|SelectFloat (Math|Trig|RadiansToDegrees (Math|Trig|Acos(Radians) (/ (Math|Vector|DotProduct Origin%20Rotation Destination%20Rotation) (* (Math|Vector|VectorLength Destination%20Rotation) (Math|Vector|VectorLength Origin%20Rotation))))) (- 3.6e+02 (Math|Trig|RadiansToDegrees (Math|Trig|Acos(Radians) (/ (Math|Vector|DotProduct Origin%20Rotation Destination%20Rotation) (* (Math|Vector|VectorLength Destination%20Rotation) (Math|Vector|VectorLength Origin%20Rotation)))))) (> (Math|Vector|DotProduct Rotation%20Normal (Math|Vector|CrossProduct Origin%20Rotation Destination%20Rotation)) 0))))

(fn CalculateDegreesTurned
  (params (A :type Float%20%28double-precision%29) (B :type Float%20%28double-precision%29)))

(fn GetLocation
  (Variables|Default|SetXY%20Vector (Math|Vector|MakeVector (.x (Transformation|GetActorLocation (Variables|GAVars|Data|GetOwningPlayerPawn))) (.y (Transformation|GetActorLocation (Variables|GAVars|Data|GetOwningPlayerPawn)))))
  (return))
